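# Physics asset for the skeletal mesh of the Let Em Cook gadget, providing its simplified collision shapes.
# It has two bodies, one on the main bone and one on the root bone, each a single capsule.
# The main capsule is rotated and slightly offset, while the smaller root capsule has no offset.
# No constraints join the two bodies.
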PhysicsAsset /Game/Gadgets/Assets/LetEmCook/Meshes/Let_Em_Cook_Gadget_Physics
bodies {
 main [{shapeType=Capsule, length=18.9855, radius=18.6972, center=<-0.1467 0.0048 0.0656>, rotation=<rot 0 178.0519 -89.059>}]
 root [{shapeType=Capsule, length=15, radius=15, shapeName=root_capsule}]
}
constraints []
